WidgetBlueprint /Game/UI/Foundation/Modals/PlayerReportActioned/PlayerActionedWindow
animations [Intro {
  binding_count=3
  display_label=""
  length_seconds=3.2
  object=&self:Intro
  sequence {
   bindings [
    SizeBox_0 {
     id=#1
     possessed_class=SizeBox
     tracks [
      {
       type=MovieScene2DTransformTrack
       sections [{
         type=MovieScene2DTransformSection
         end_frame=192
         start_frame=0
         channels [
          Translation.X {type=MovieSceneScriptingFloatChannel}
          Translation.Y {type=MovieSceneScriptingFloatChannel}
          Angle {type=MovieSceneScriptingFloatChannel}
          Scale.X {type=MovieSceneScriptingFloatChannel, keys [[15,1.3],[25.639,0.9507],[28.947,1]]}
          Scale.Y {type=MovieSceneScriptingFloatChannel, keys [[15,1.3],[25.639,0.9507],[28.947,1]]}
          Shear.X {type=MovieSceneScriptingFloatChannel}
          Shear.Y {type=MovieSceneScriptingFloatChannel}
         ]
        }]
      }
      {
       type=MovieSceneFloatTrack
       display_name=Render Opacity
       sections [{type=MovieSceneFloatSection, end_frame=192, start_frame=0, channels [None {type=MovieSceneScriptingFloatChannel, keys [[15,0],[24,1]]}]}]
      }
     ]
    }
    BackgroundBlur_0 {
     id=#2
     possessed_class=BackgroundBlur
     tracks [{
       type=MovieSceneFloatTrack
       display_name=Blur Strength
       sections [{type=MovieSceneFloatSection, end_frame=12, start_frame=0, channels [None {type=MovieSceneScriptingFloatChannel, keys [[0,0],[12,50]]}]}]
      }]
    }
    Image_34 {
     id=#3
     possessed_class=Image
     tracks [
      {
       type=MovieScene2DTransformTrack
       sections [{
         type=MovieScene2DTransformSection
         end_frame=117
         start_frame=0
         channels [
          Translation.X {type=MovieSceneScriptingFloatChannel, const [-27,1]}
          Translation.Y {type=MovieSceneScriptingFloatChannel, const [0,1]}
          Angle {type=MovieSceneScriptingFloatChannel}
          Scale.X {type=MovieSceneScriptingFloatChannel, keys [[9,0.8],[18,1]]}
          Scale.Y {type=MovieSceneScriptingFloatChannel, keys [[9,0.8],[18,1]]}
          Shear.X {type=MovieSceneScriptingFloatChannel}
          Shear.Y {type=MovieSceneScriptingFloatChannel}
         ]
        }]
      }
      {
       type=MovieSceneFloatTrack
       display_name=Render Opacity
       sections [{type=MovieSceneFloatSection, end_frame=31, start_frame=0, channels [None {type=MovieSceneScriptingFloatChannel, keys [[9,0],[18,1]]}]}]
      }
     ]
    }
   ]
   display_rate {numerator=60}
  }
 }]
graphs [EventGraph {nodes [
   "On Button Clicked (CloseButton)"!() -> (OutputDelegate, then, Button:Common Button Legacy Object Reference) {is=title K2Node_ComponentBoundEvent}
   > DeactivateWidget()
   Event On Deactivated()
   > Call On Message Closed!(execute:Exec, self:Player Actioned Window Object Reference) -> (then:Exec) {is=title K2Node_CallDelegate}
   n1 = Event Construct()
   PlayAnimation(execute=[@n1.then, @n2.then], InAnimation=Get Intro())
   n2 = Event On Activated()
  ]}]
variables []
widgets {
 info {
  inheritedWidgetCount=1
  namedSlotCount=1
  parentClass=/Script/FortniteUI.FortMessageReceivedScreen
  rootWidgetClass=&Overlay
  widgetCount=18
 }
 widgets [
  Overlay_182 {widgetClassPath=&Overlay, widget=Overlay_182}
  BackgroundBlur_0 {
   widgetClassPath=&BackgroundBlur
   parent=Overlay_182
   widget=BackgroundBlur_0
   properties {blurStrength=50}
   slot {class=OverlaySlot, properties {horizontalAlignment=HAlign_Fill, verticalAlignment=VAlign_Fill}}
  }
  Lightbox {
   widgetClassPath=/Game/UI/Foundation/Lightbox/Lightbox
   bIsVariable=true
   parent=BackgroundBlur_0
   widget=Lightbox
   properties {clipping=ClipToBounds}
   slot {class=BackgroundBlurSlot, properties {padding=<ltrb 0 0 0 0>}}
  }
  Overlay_183 {widgetClassPath=&Overlay, namedSlotHost=Lightbox, widget=Overlay_183}
  Image_34 {
   widgetClassPath=&Image
   bIsVariable=true
   parent=Overlay_183
   widget=Image_34
   properties {
    brush {
     bIsDynamicallyLoaded=false
     drawAs=Image
     imageType=NoImage
     mirroring=NoMirror
     resourceName=None
     resourceObject=&/T_ModalTeddy_1024
     tiling=NoTile
     imageSize=<1202 1024>
     margin=<ltrb 0 0 0 0>
     outlineSettings {
      bUseBrushTransparency=false
      roundingType=HalfHeightRadius
      width=0
      color {colorUseRule=UseColor_Specified, specifiedColor=<rgba 0 0 0 0>}
      cornerRadii=<0 0 0 0>
     }
     tintColor {colorUseRule=UseColor_Specified, specifiedColor=<rgba 1 1 1 1>}
     uVRegion {bIsValid=false, max=<0 0>, min=<0 0>}
    }
    renderTransform {angle=0, scale=<1 1>, shear=<0 0>, translation=<-79 49>}
    renderTransformPivot=<0.5 1>
   }
   slot {class=OverlaySlot, properties {verticalAlignment=VAlign_Bottom}}
  }
  SizeBox_0 {
   widgetClassPath=&SizeBox
   parent=Overlay_183
   widget=SizeBox_0
   properties {bOverride_WidthOverride=true, heightOverride=350, widthOverride=1000, renderTransformPivot=<0.25 0.75>}
   slot {class=OverlaySlot, properties {horizontalAlignment=HAlign_Right, verticalAlignment=VAlign_Center, padding=<ltrb 0 0 90 0>}}
  }
  BackgroundBlur_1 {widgetClassPath=&BackgroundBlur, parent=SizeBox_0, widget=BackgroundBlur_1, properties {blurStrength=5}, slot {class=SizeBoxSlot}}
  CommonBorder_0 {
   widgetClassPath=&CommonBorder
   parent=BackgroundBlur_1
   widget=CommonBorder_0
   properties {
    style=/FortUILibrary/Styles/Border/Border_Navy_VGrad.Border_Navy_VGrad_C
    background {
     bIsDynamicallyLoaded=false
     drawAs=Image
     imageType=NoImage
     mirroring=NoMirror
     resourceName=None
     resourceObject=/FortUILibrary/Styles/Border/Materials/M_Border_Navy_VGrad
     tiling=NoTile
     imageSize=<32 32>
     margin=<ltrb 0 0 0 0>
     outlineSettings {
      bUseBrushTransparency=false
      roundingType=HalfHeightRadius
      width=0
      color {colorUseRule=UseColor_Specified, specifiedColor=<rgba 0 0 0 0>}
      cornerRadii=<0 0 0 0>
     }
     tintColor {colorUseRule=UseColor_Specified, specifiedColor=<rgba 1 1 1 1>}
     uVRegion {bIsValid=false, max=<0 0>, min=<0 0>}
    }
    brushColor=<rgba 1 1 1 0.5714>
    padding=<ltrb 0 0 0 0>
   }
   slot {class=BackgroundBlurSlot, properties {padding=<ltrb 0 0 0 0>}}
  }
  VerticalBox_0 {
   widgetClassPath=&VerticalBox
   parent=CommonBorder_0
   widget=VerticalBox_0
   slot {class=BorderSlot, properties {padding=<ltrb 0 0 0 0>}}
  }
  CommonBorder_Modal {
   widgetClassPath=&CommonBorder
   parent=VerticalBox_0
   widget=CommonBorder_Modal
   properties {style=/FortUILibrary/Styles/Border/Border-ModalHeader.Border-ModalHeader_C, brushColor=<rgba 1 1 1 0.6762>}
   slot {class=VerticalBoxSlot}
  }
  HorizontalBox_1 {widgetClassPath=&HorizontalBox, parent=CommonBorder_Modal, widget=HorizontalBox_1, slot {class=BorderSlot}}
  Spacer_0 {widgetClassPath=&Spacer, parent=HorizontalBox_1, widget=Spacer_0, properties {size=<44 100>}, slot {class=HorizontalBoxSlot}}
  Title {
   widgetClassPath=&CommonTextBlock
   bIsVariable=true
   parent=HorizontalBox_1
   widget=Title
   properties {
    style=/FortUILibrary/Styles/Typography/Foundation/TextStyle-Header-L.TextStyle-Header-L_C
    text=REPORT ANALYZED
    font {
     bForceMonospaced=false
     bMaterialIsStencil=false
     fontMaterial=null
     fontObject=&Burbank
     letterSpacing=0
     monospacedWidth=1
     size=48
     skewAmount=0
     typefaceFontName=Burbank-Black
     outlineSettings {
      bApplyOutlineToDropShadows=false
      bMiteredCorners=false
      bSeparateFillAlpha=false
      outlineMaterial=null
      outlineSize=0
      outlineColor=<rgba 0 0 0 1>
     }
    }
    shadowOffset=<0 0>
   }
   slot {class=HorizontalBoxSlot, properties {horizontalAlignment=HAlign_Left, verticalAlignment=VAlign_Center, padding=<ltrb 8 0 0 0>}}
  }
  VerticalBox_1 {
   widgetClassPath=&VerticalBox
   parent=VerticalBox_0
   widget=VerticalBox_1
   slot {class=VerticalBoxSlot, properties {verticalAlignment=VAlign_Center, padding=<ltrb 50 50 50 50>, size {sizeRule=Fill, value=1}}}
  }
  CommonTextBlock_1 {
   widgetClassPath=&CommonTextBlock
   parent=VerticalBox_1
   widget=CommonTextBlock_1
   properties {
    style=/FortUILibrary/Styles/Typography/Foundation/TextStyle-Base-M-I.TextStyle-Base-M-I_C
    text=Offender is not a bear.
    font {
     bForceMonospaced=false
     bMaterialIsStencil=false
     fontMaterial=null
     fontObject=&NotoSans
     letterSpacing=0
     monospacedWidth=1
     size=24
     skewAmount=0
     typefaceFontName=NotoSans-Italic
     outlineSettings {
      bApplyOutlineToDropShadows=false
      bMiteredCorners=false
      bSeparateFillAlpha=false
      outlineMaterial=null
      outlineSize=0
      outlineColor=<rgba 0 0 0 1>
     }
    }
    shadowOffset=<0 0>
   }
   slot {class=VerticalBoxSlot, properties {padding=<ltrb 0 0 0 40>}}
  }
  CommonTextBlock_0 {
   widgetClassPath=&CommonTextBlock
   parent=VerticalBox_1
   widget=CommonTextBlock_0
   properties {
    style=/FortUILibrary/Styles/Typography/Foundation/TextStyle-Base-M.TextStyle-Base-M_C
    text="Thanks to your report, we took action on another player's negative behavior. \r\n\r\nThanks for helping to keep Fortnite a fun, friendly, and inclusive community!"
    wrapTextAt=876.6393
    font {
     bForceMonospaced=false
     bMaterialIsStencil=false
     fontMaterial=null
     fontObject=&NotoSans
     letterSpacing=0
     monospacedWidth=1
     size=24
     skewAmount=0
     typefaceFontName=NotoSans-Regular
     outlineSettings {
      bApplyOutlineToDropShadows=false
      bMiteredCorners=false
      bSeparateFillAlpha=false
      outlineMaterial=null
      outlineSize=0
      outlineColor=<rgba 0 0 0 1>
     }
    }
    shadowOffset=<0 0>
   }
   slot {class=VerticalBoxSlot}
  }
  CloseButton {
   widgetClassPath=&IconTextButton
   bIsVariable=true
   parent=VerticalBox_0
   widget=CloseButton
   properties {
    bHideInputActionWithKeyboard=true
    buttonText=CLOSE
    contentAlignment=HAlign_Center
    minHeight=90
    minWidth=319
    onButtonBaseClicked=(&self:WidgetTree.CloseButton.HandleOnButtonClicked)
    onButtonBaseDoubleClicked=(&self:WidgetTree.CloseButton.HandleOnButtonDoubleClicked)
    onButtonBaseHovered=(&self:WidgetTree.CloseButton.HandleOnButtonHovered)
    onButtonBaseUnhovered=(&self:WidgetTree.CloseButton.HandleOnButtonUnhovered)
    onSelectedChangedBase=(&self:WidgetTree.CloseButton.HandleOnSelectedChanged)
    style=/FortUILibrary/Foundation/Buttons/ButtonStyles/ButtonStyle_Feature_L.ButtonStyle_Feature_L_C
    triggeringInputAction {dataTable=&InputActionDataTable, rowName=Continue}
   }
   slot {class=VerticalBoxSlot, properties {horizontalAlignment=HAlign_Center, verticalAlignment=VAlign_Bottom, padding=<ltrb 20 20 20 20>}}
  }
  GiftBoxImage {widgetClassPath=/Script/CommonUI.CommonLazyImage, bInherited=true, widget=null}
 ]
}
